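# Dizzy emote effect for the PouchGnaw cosmetic companion: stars and spinning wind meshes circle above the effect origin.
# Skeletal_Wind spawns meshes at three partition-based tilts, spins them with a waveform wobble and fades their alpha in and out.
# Stars and smaller Stars002 sprites orbit by re-evaluating sphere ShapeLocation each update, using emitter age for angle and waveforms for height, with camera offset and size curves.
# Skeletal_Wind002, Stars001 and Stars003 are disabled variants.
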NiagaraSystem /CosmeticCompanions/FX/Builtins/PouchGnaw/NS_Emote_PouchGnaw_Dizzy
dependencies {usedDataInterfaces [&NiagaraDataInterfaceCurve], usedDynamicInputs [], usedModules [&SystemState]}
emitters {
 Skeletal_Wind {
  emitterSpawnScript [/SetVariables(Emitter.RandomSplineWith:NiagaraFloat=&RandomRangeFloat(Minimum=1, Maximum=10)) {isSetParameters=true}]
  emitterUpdateScript [
   &EmitterState(
    Life Cycle Mode=::NewEnumerator1
    Loop Behavior=::NewEnumerator1
    Loop Delay=0.1
    Loop Duration=1.5
    Scalability Mode=::NewEnumerator1
    UseLoopDelay=false
   )
   &SpawnRate(SpawnRate=2.5, Use Spawn Probability=false)
   &SpawnBurst_Instantaneous(Spawn Count=2, Use Spawn Probability=false, UseLoopCountLimit=false)
  ]
  particleSpawnScript [
   &PartitionParticles(Number Of Partitions=3)
   &InitializeParticle(
    Color=<rgba 1 1 1 2>
    Mesh Scale=&Multiply_Vector(
     A=&VectorFromFloat(Value=&RandomRangeFloat(Minimum=0.75, Maximum=1.05))
     B=&RandomRangeVector(Minimum=<0.2 0.01 0.2>, Maximum=<0.15 0.18 0.15>)
    )
    Mesh Scale Mode=::NewEnumerator4
    Position=<0 0 28>
    Position Mode=::NewEnumerator1
    UsePositionOffset=false
    Write Material Random=true
   )
   &InitialMeshOrientation(
    Mesh Axis To Orient=<0 1 0>
    Mesh Orientation Mode=::NewEnumerator2
    Orientation Coordinate Space=::World
    Orientation Vector=<0 0 1>
    Rotation=&MakeVector(
     X=&MergeHelper_FloatSwitch(0=-1.0371, 1=0, 2=0.9836, 3=?, Index=$Particles.PartitionParticles.Partition, Num=::NewEnumerator1)
     Y=&RandomRangeFloat(Minimum=0, Maximum=0.4055)
     Z=0
    )
    Use Rotation Vector=true
   )
  ]
  particleUpdateScript [
   &ParticleState()
   &UpdateMeshOrientation(
    Rotation Coordinate Space=::NewEnumerator0
    Rotation Vector=&MakeVector(
     X=&Waveform("Waveform [1]"=::NewEnumerator0, "[1] Frequency"=1, "[1] Amplitude Min/Max"=<-1 1>, "[1] Scale"=0.15)
     Y=0
     Z=&RandomRangeFloat(Minimum=0.75, Maximum=1.45)
    )
   )
   &Drag(
    Drag=&FloatFromCurve(
     FloatCurve=DataInterface() {propertyValues {Curve {defaultValue=1, keys [{tangentMode=RCTM_User, time=0, value=0}, {tangentMode=RCTM_User, time=1, value=1}]}}}
     Scale Curve=2
    )
    Use Linear Drag=true
    Use Rotational Drag=true
   ) {enabled=false}
   &SolveForcesAndVelocity(Clamp Velocity=false, Limit Acceleration=false)
   &ScaleColor(
    Scale Alpha=&FloatFromCurve(FloatCurve=DataInterface() {propertyValues {Curve {keys [
         {arriveTangent=2.4536, interpMode=RCIM_Cubic, leaveTangent=2.4536, tangentMode=RCTM_User, time=0.017, value=0}
         {arriveTangent=0.6923, interpMode=RCIM_Cubic, leaveTangent=0.6923, time=0.15, value=0.5}
         {interpMode=RCIM_Cubic, tangentMode=RCTM_User, time=0.7297, value=0.4934}
         {arriveTangent=-2.5865, tangentMode=RCTM_User, time=1, value=0}
        ]}}})
    ScaleA=true
    ScaleRGB=false
   )
  ]
  rendererClasses [&NiagaraMeshRendererProperties]
  renderers [{rendererClass=&NiagaraMeshRendererProperties, rendererIndex=0}]
 }
 Skeletal_Wind002 {
  bEnabled=false
  emitterUpdateScript [
   &EmitterState(
    Life Cycle Mode=::NewEnumerator1
    Loop Behavior=::NewEnumerator1
    Loop Delay=0.1
    Loop Duration=2
    Scalability Mode=::NewEnumerator1
    UseLoopDelay=false
   )
   &SpawnBurst_Instantaneous(Spawn Count=3, Use Spawn Probability=false, UseLoopCountLimit=false)
  ]
  particleSpawnScript [
   &PartitionParticles(Number Of Partitions=3)
   &InitializeParticle(
    Color=<rgba 1 1 1 2>
    Lifetime=2
    Mesh Scale=&Multiply_Vector(
     A=&VectorFromFloat(Value=&RandomRangeFloat(Minimum=1.15, Maximum=1.35))
     B=&RandomRangeVector(Minimum=<0.2 0.03 0.2>, Maximum=<0.15 0.02 0.15>)
    )
    Mesh Scale Mode=::NewEnumerator4
    Position=<0 0 25>
    Position Mode=::NewEnumerator1
    UsePositionOffset=false
    Write Material Random=true
   )
   &InitialMeshOrientation(
    Mesh Axis To Orient=<0 1 0>
    Mesh Orientation Mode=::NewEnumerator2
    Orientation Coordinate Space=::World
    Orientation Vector=<0 0 1>
    Rotation=&MakeVector(
     X=&MergeHelper_FloatSwitch(0=-1.0371, 1=0, 2=0.9836, 3=?, Index=$Particles.PartitionParticles.Partition, Num=::NewEnumerator1)
     Y=&RandomRangeFloat(Minimum=0, Maximum=0.4055)
     Z=0
    )
    Use Rotation Vector=true
   )
  ]
  particleUpdateScript [
   &ParticleState()
   &UpdateMeshOrientation(
    Rotation Coordinate Space=::NewEnumerator0
    Rotation Vector=&MakeVector(
     X=&Waveform("Waveform [1]"=::NewEnumerator0, "[1] Frequency"=1, "[1] Amplitude Min/Max"=<-1 1>, "[1] Scale"=0.15)
     Y=0
     Z=&RandomRangeFloat(Minimum=0.25, Maximum=1)
    )
   )
   &Drag(
    Drag=&FloatFromCurve(
     FloatCurve=DataInterface() {propertyValues {Curve {defaultValue=1, keys [{tangentMode=RCTM_User, time=0, value=0}, {tangentMode=RCTM_User, time=1, value=1}]}}}
     Scale Curve=2
    )
    Use Linear Drag=true
    Use Rotational Drag=true
   ) {enabled=false}
   &SolveForcesAndVelocity(Clamp Velocity=false, Limit Acceleration=false)
   &ScaleColor(
    Scale Alpha=&FloatFromCurve(FloatCurve=DataInterface() {propertyValues {Curve {keys [
         {arriveTangent=2.4536, interpMode=RCIM_Cubic, leaveTangent=2.4536, tangentMode=RCTM_User, time=0.017, value=0}
         {arriveTangent=0.6002, interpMode=RCIM_Cubic, leaveTangent=0.6002, time=0.15, value=0.5}
         {interpMode=RCIM_Cubic, tangentMode=RCTM_User, time=0.85, value=0.5}
         {arriveTangent=-2.5865, tangentMode=RCTM_User, time=1, value=0}
        ]}}})
    ScaleA=true
    ScaleRGB=false
   )
  ]
  rendererClasses [&NiagaraMeshRendererProperties]
  renderers [{rendererClass=&NiagaraMeshRendererProperties, rendererIndex=0}]
 }
 Stars {
  emitterUpdateScript [
   &EmitterState(Life Cycle Mode=::NewEnumerator1, Loop Behavior=::NewEnumerator1, Loop Duration=1.1, UseLoopDelay=false)
   &SpawnRate(SpawnRate=4.8, Use Spawn Probability=false)
   &SpawnBurst_Instantaneous(Spawn Count=2, Use Spawn Probability=false, UseLoopCountLimit=false)
  ]
  particleSpawnScript [&InitializeParticle(
    Color Mode=::NewEnumerator0
    Lifetime=&RandomRangeFloat(Minimum=1, Maximum=1.36)
    Sprite Size Mode=::NewEnumerator3
    Uniform Sprite Size=&RandomRangeFloat(Minimum=2, Maximum=10)
    UsePositionOffset=false
    Write Material Random=true
   )]
  particleUpdateScript [
   &ParticleState()
   &ShapeLocation(
    MaskSpawnGroup=false
    Shape Origin=&AddVectorToPosition(Position=$Engine.Emitter.SimulationPosition, Vector=<0 0 28>)
    Sphere Distribution=::NewEnumerator1
    Sphere Radius=25
    Sphere Radius Position:NiagaraFloat=&RandomRangeFloat(Minimum=0.7, Maximum=1)
    U Position=&Divide_Float(
     A=&Add_Float(A=$Emitter.Age, B=&RandomRangeFloat(Minimum=-12.5, Maximum=12.5))
     B=&RandomRangeFloat(Minimum=0.375, Maximum=0.65)
    )
    UseOwnerScale=true
    V Position=&Waveform(
     "Waveform [1]"=::NewEnumerator0
     "[1] Frequency"=&RandomRangeFloat(Minimum=0.8, Maximum=1.2)
     "[1] Amplitude Min/Max"=<-1 1>
     "[1] Scale"=&RandomRangeFloat(Minimum=-0.112, Maximum=0.112)
     "[1] Bias"=0.5
    )
   )
   &CameraOffset(Camera Offset Amount=20)
   &ScaleSpriteSize(Uniform Curve Sprite Scale=DataInterface() {propertyValues {Curve {
       defaultValue=0
       keys [
        {interpMode=RCIM_Cubic, tangentMode=RCTM_User, time=0, value=0}
        {arriveTangent=-0.0908, interpMode=RCIM_Cubic, leaveTangent=-0.0908, tangentMode=RCTM_User, time=0.15, value=1}
        {interpMode=RCIM_Cubic, tangentMode=RCTM_User, time=0.65, value=1}
        {tangentMode=RCTM_User, time=1, value=0}
       ]
      }}})
  ]
  rendererClasses [&NiagaraSpriteRendererProperties]
  renderers [{rendererClass=&NiagaraSpriteRendererProperties}]
 }
 Stars001 {
  bEnabled=false
  emitterUpdateScript [
   &EmitterState(Life Cycle Mode=::NewEnumerator1, Loop Behavior=::NewEnumerator1, Loop Duration=1.1, UseLoopDelay=false)
   &SpawnRate(SpawnRate=4, Use Spawn Probability=false)
   &SpawnBurst_Instantaneous(Spawn Count=2, Use Spawn Probability=false, UseLoopCountLimit=false)
  ]
  particleSpawnScript [&InitializeParticle(
    Color Mode=::NewEnumerator0
    Lifetime=&RandomRangeFloat(Minimum=1, Maximum=1.36)
    Sprite Size Mode=::NewEnumerator3
    Uniform Sprite Size=&RandomRangeFloat(Minimum=2, Maximum=10)
    UsePositionOffset=false
    Write Material Random=true
   )]
  particleUpdateScript [
   &ParticleState()
   &ShapeLocation(
    MaskSpawnGroup=false
    Shape Origin=&AddVectorToPosition(Position=$Engine.Emitter.SimulationPosition, Vector=<0 0 28>)
    Sphere Distribution=::NewEnumerator1
    Sphere Radius=25
    Sphere Radius Position:NiagaraFloat=&RandomRangeFloat(Minimum=0.7, Maximum=1)
    U Position=&Divide_Float(
     A=&Add_Float(A=$Emitter.Age, B=&RandomRangeFloat(Minimum=-12.5, Maximum=12.5))
     B=&RandomRangeFloat(Minimum=0.7, Maximum=1.3)
    )
    UseOwnerScale=true
    V Position=&Waveform(
     "Waveform [1]"=::NewEnumerator0
     "[1] Frequency"=&RandomRangeFloat(Minimum=0.8, Maximum=1.2)
     "[1] Amplitude Min/Max"=<-1 1>
     "[1] Scale"=&RandomRangeFloat(Minimum=-0.112, Maximum=0.112)
     "[1] Bias"=0.5
    )
   )
   &CameraOffset(Camera Offset Amount=20)
   &ScaleSpriteSize(Uniform Curve Sprite Scale=DataInterface() {propertyValues {Curve {
       defaultValue=0
       keys [
        {interpMode=RCIM_Cubic, tangentMode=RCTM_User, time=0, value=0}
        {arriveTangent=-0.0908, interpMode=RCIM_Cubic, leaveTangent=-0.0908, tangentMode=RCTM_User, time=0.15, value=1}
        {interpMode=RCIM_Cubic, tangentMode=RCTM_User, time=0.65, value=1}
        {tangentMode=RCTM_User, time=1, value=0}
       ]
      }}})
  ]
  rendererClasses [&NiagaraSpriteRendererProperties]
  renderers [{rendererClass=&NiagaraSpriteRendererProperties}]
 }
 Stars002 {
  emitterUpdateScript [
   &EmitterState(Life Cycle Mode=::NewEnumerator1, Loop Behavior=::NewEnumerator1, Loop Duration=1.35, UseLoopDelay=false)
   &SpawnRate(SpawnRate=4, Use Spawn Probability=false)
   &SpawnBurst_Instantaneous(Spawn Count=2, Use Spawn Probability=false, UseLoopCountLimit=false)
  ]
  particleSpawnScript [&InitializeParticle(
    Color Mode=::NewEnumerator0
    Lifetime=&RandomRangeFloat(Minimum=0.8, Maximum=1)
    Sprite Size Mode=::NewEnumerator3
    Uniform Sprite Size=&RandomRangeFloat(Minimum=2, Maximum=3)
    UsePositionOffset=false
    Write Material Random=true
   )]
  particleUpdateScript [
   &ParticleState()
   &ShapeLocation(
    MaskSpawnGroup=false
    Shape Origin=&AddVectorToPosition(Position=$Engine.Emitter.SimulationPosition, Vector=<0 0 28>)
    Sphere Distribution=::NewEnumerator1
    Sphere Radius=25
    Sphere Radius Position:NiagaraFloat=&RandomRangeFloat(Minimum=0.6, Maximum=1)
    U Position=&Divide_Float(
     A=&Add_Float(A=$Emitter.Age, B=&RandomRangeFloat(Minimum=-12.5, Maximum=12.5))
     B=&RandomRangeFloat(Minimum=0.5, Maximum=0.825)
    )
    UseOwnerScale=true
    V Position=&Waveform(
     "Waveform [1]"=::NewEnumerator0
     "[1] Frequency"=&RandomRangeFloat(Minimum=0.6, Maximum=1)
     "[1] Amplitude Min/Max"=<-1 1>
     "[1] Scale"=&RandomRangeFloat(Minimum=-0.1, Maximum=0.1)
     "[1] Bias"=0.5
    )
   )
   &CameraOffset(Camera Offset Amount=20)
   &ScaleSpriteSize(Uniform Curve Sprite Scale=DataInterface() {propertyValues {Curve {
       defaultValue=0
       keys [
        {interpMode=RCIM_Cubic, tangentMode=RCTM_User, time=0, value=0}
        {arriveTangent=-0.0908, interpMode=RCIM_Cubic, leaveTangent=-0.0908, tangentMode=RCTM_User, time=0.15, value=1}
        {interpMode=RCIM_Cubic, tangentMode=RCTM_User, time=0.75, value=1}
        {tangentMode=RCTM_User, time=1, value=0}
       ]
      }}})
  ]
  rendererClasses [&NiagaraSpriteRendererProperties]
  renderers [{rendererClass=&NiagaraSpriteRendererProperties}]
 }
 Stars003 {
  bEnabled=false
  emitterUpdateScript [
   &EmitterState(Life Cycle Mode=::NewEnumerator1, Loop Behavior=::NewEnumerator1, Loop Duration=1.35, UseLoopDelay=false)
   &SpawnRate(SpawnRate=4, Use Spawn Probability=false)
   &SpawnBurst_Instantaneous(Spawn Count=2, Use Spawn Probability=false, UseLoopCountLimit=false)
  ]
  particleSpawnScript [&InitializeParticle(
    Color Mode=::NewEnumerator0
    Lifetime=&RandomRangeFloat(Minimum=0.8, Maximum=1)
    Sprite Size Mode=::NewEnumerator3
    Uniform Sprite Size=&RandomRangeFloat(Minimum=2, Maximum=3)
    UsePositionOffset=false
    Write Material Random=true
   )]
  particleUpdateScript [
   &ParticleState()
   &ShapeLocation(
    MaskSpawnGroup=false
    Shape Origin=&AddVectorToPosition(Position=$Engine.Emitter.SimulationPosition, Vector=<0 0 28>)
    Sphere Distribution=::NewEnumerator1
    Sphere Radius=25
    Sphere Radius Position:NiagaraFloat=&RandomRangeFloat(Minimum=0.6, Maximum=1)
    U Position=&Divide_Float(
     A=&Add_Float(A=$Emitter.Age, B=&RandomRangeFloat(Minimum=-12.5, Maximum=12.5))
     B=&RandomRangeFloat(Minimum=0.9, Maximum=1.45)
    )
    UseOwnerScale=true
    V Position=&Waveform(
     "Waveform [1]"=::NewEnumerator0
     "[1] Frequency"=&RandomRangeFloat(Minimum=0.6, Maximum=1)
     "[1] Amplitude Min/Max"=<-1 1>
     "[1] Scale"=&RandomRangeFloat(Minimum=-0.1, Maximum=0.1)
     "[1] Bias"=0.5
    )
   )
   &CameraOffset(Camera Offset Amount=20)
   &ScaleSpriteSize(Uniform Curve Sprite Scale=DataInterface() {propertyValues {Curve {
       defaultValue=0
       keys [
        {interpMode=RCIM_Cubic, tangentMode=RCTM_User, time=0, value=0}
        {arriveTangent=-0.0908, interpMode=RCIM_Cubic, leaveTangent=-0.0908, tangentMode=RCTM_User, time=0.15, value=1}
        {interpMode=RCIM_Cubic, tangentMode=RCTM_User, time=0.75, value=1}
        {tangentMode=RCTM_User, time=1, value=0}
       ]
      }}})
  ]
  rendererClasses [&NiagaraSpriteRendererProperties]
  renderers [{rendererClass=&NiagaraSpriteRendererProperties}]
 }
}
summary {
 systemName=NS_Emote_PouchGnaw_Dizzy
 emitters [
  {emitterName=Stars, rendererClasses [&NiagaraSpriteRendererProperties]}
  {emitterName=Skeletal_Wind, rendererClasses [&NiagaraMeshRendererProperties]}
  {bEnabled=false, emitterName=Skeletal_Wind002, rendererClasses [&NiagaraMeshRendererProperties]}
  {emitterName=Stars002, rendererClasses [&NiagaraSpriteRendererProperties]}
  {bEnabled=false, emitterName=Stars001, rendererClasses [&NiagaraSpriteRendererProperties]}
  {bEnabled=false, emitterName=Stars003, rendererClasses [&NiagaraSpriteRendererProperties]}
 ]
 userVariables []
}
